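# Niagara ribbon beam for the Umbramolt rift gate, strung between two user-supplied spline endpoints.
# A partitioned burst places particles along a Hermite spline, adding an animated sine/cosine spiral offset that tapers to zero at both ends.
# OpenAmount scales width and alpha profiles along the beam, while RemoveBeam drives its dissolve-out.
# Endpoint distance over 1280 units drives thinning, fading and random brightness flicker; time of day also scales alpha.
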NiagaraSystem /Umbramolt_Hole/Gameplay/FX/NS_RiftGate_Umbramolt_Beam
dependencies {usedDataInterfaces [&NiagaraDataInterfaceCurve], usedDynamicInputs [], usedModules [&SystemState]}
emitters {SplineElectric {
  emitterSpawnScript [/Game/Effects/Niagara/Modules/Partition/FN_PartitionAttributes(Number of Bone Sets=1, Particles Per Partition=12)]
  emitterUpdateScript [
   &EmitterState(Life Cycle Mode=::NewEnumerator1, Loop Behavior=::NewEnumerator1, Loop Duration Mode=::NewEnumerator1, UseLoopDelay=false)
   &SpawnBurst_Instantaneous(Spawn Count=$Emitter.FN_PartitionAttributes.Spawn Total, Use Spawn Probability=false)
  ]
  particleSpawnScript [
   /Game/Effects/Niagara/Modules/Partition/FN_Particle_partition_v5/FN_Particle_Partition_v5(Enable Ribbon ID=true)
   &InitializeRibbon(
    Color=<rgba 3 3 3 0.7>
    Module.Write Color=true
    Module.Write Lifetime=true
    Module.Write Mass=true
    Module.Write Material Random=true
    Module.Write Position=false
    Module.Write RibbonTwist=false
    Module.Write RibbonWidth=true
    Ribbon Width=$User.OverallWidth
   )
   /SetVariables(Particles.BeamEndDissolve=1) {isSetParameters=true}
  ]
  particleUpdateScript [
   &ParticleState(Kill Particles When Lifetime Has Elapsed=false, Let Infinitely Lived Particles Die When Emitter Deactivates=true)
   /SetVariables(
    Particles.BeamEndDissolve:NiagaraFloat=&ClampFloat(Float=&Subtract_Float(
      A=$Particles.BeamEndDissolve
      B=&Multiply_Float(
       A=$Engine.DeltaTime
       B=&MakeCustomFloatFromBool(
        Bool=&SetBoolByFloatComparison(A=$User.RemoveBeam, B=1, Comparison Type=::NewEnumerator4)
        True Float=4
        False Float=0
       )
      )
     ))
    Particles.BeamStartDissolve:NiagaraFloat=&ClampFloat(Float=&Add_Float(A=$Particles.BeamStartDissolve, B=&Multiply_Float(A=$Engine.DeltaTime, B=2)))
   ) {isSetParameters=true}
   &DynamicMaterialParameters(
    Index 0 Param 1=&FloatFromCurve(
     FloatCurve=DataInterface() {propertyValues {Curve {keys [{tangentMode=RCTM_User, time=0, value=0}, {tangentMode=RCTM_User, time=1, value=1}]}}}
     CurveIndex=&Multiply_Float(A=$Particles.BeamStartDissolve, B=$Particles.BeamEndDissolve)
    )
    Index 0 Param 2=1.02
    Module.Param0WriteEnabled=true
    Module.Param1WriteEnabled=true
    Module.Param2WriteEnabled=false
    Module.Param3WriteEnabled=false
   )
   &SystemLocation(MaskSpawnGroup=false)
   /Game/Effects/Niagara/Modules/Curves/FN_HermiteCubicSpline_v1/FN_HermiteCubicSpline(
    Position Blend:NiagaraFloat=$Particles.Initial.Partition.NormalizedID
    Position End:Vector3f=$User.SplineUpdatedStartPos
    Position Start:Vector3f=$User.SplineUpdatedEndPos
    Write Particles Position=true
   )
   &self:WaveAxis_02(EndPos:Vector3f=$User.SplineUpdatedEndPos, StartPos:Vector3f=$User.SplineUpdatedStartPos) {module=WaveAxis_02}
   /SetVariables(Particles.SinePosition:Vector3f=&Add_Vector(
     A=&Multiply_Vector(
      A=$Particles.Y
      B=&VectorFromFloat(Value=&Sine(
        Normalized Angle=&Add_Float(
         A=&Multiply_Float(A=$Particles.Partition.NormalizedID, B=-1)
         B=&Multiply_Float(A=$Engine.Time, B=0.2)
        )
        Period=0.5
        Scale=15
        Bias=0
       ))
     )
     B=&Multiply_Vector(
      A=$Particles.Z
      B=&VectorFromFloat(Value=&Cosine(
        Normalized Angle=&Add_Float(
         A=&Multiply_Float(A=$Particles.Partition.NormalizedID, B=-1)
         B=&Multiply_Float(A=$Engine.Time, B=0.2)
        )
        Period=0.5
        Scale=15
        Bias=0
       ))
     )
    )) {isSetParameters=true}
   /SetVariables(Particles.MultiplySinePosition:Vector3f=&Multiply_Vector(
     A=$Particles.SinePosition
     B=&VectorFromFloat(Value=&FloatFromCurve(
       FloatCurve=DataInterface() {propertyValues {Curve {keys [
           {interpMode=RCIM_Cubic, tangentMode=RCTM_User, time=0, value=0}
           {interpMode=RCIM_Cubic, tangentMode=RCTM_User, time=0.5, value=1}
           {tangentMode=RCTM_User, time=1, value=0}
          ]}}}
       CurveIndex=$Particles.Partition.NormalizedID
      ))
    )) {isSetParameters=true}
   /SetVariables(Particles.Position=&Add_Vector(A=$Particles.MultiplySinePosition, B=$Output.FN_HermiteCubicSpline.Position)) {isSetParameters=true}
   /SetVariables(Particles.EscapeDistance:NiagaraFloat=&Divide_Float(
     A=&VectorLength(Vector=&Subtract_Vector(A=$User.SplineUpdatedEndPos, B=$User.SplineUpdatedStartPos))
     B=1280
    )) {isSetParameters=true}
   &ScaleRibbonWidth(Ribbon Width Scale=&FloatFromCurve(
     FloatCurve=DataInterface() {propertyValues {Curve {keys [{tangentMode=RCTM_User, time=0, value=0}, {tangentMode=RCTM_User, time=1, value=1}]}}}
     CurveIndex=$Particles.BeamEndDissolve
    )) {module=ScaleRibbonWidth002}
   &ScaleRibbonWidth(Ribbon Width Scale=&FloatFromCurve(
     FloatCurve=DataInterface() {propertyValues {Curve {keys [
         <at 0 1 RCIM_Cubic>
         <at 0.2922 1.0544 RCIM_Cubic>
         {arriveTangent=-0.2548, interpMode=RCIM_Cubic, leaveTangent=-0.2548, time=0.7, value=1}
         {arriveTangent=-1, interpMode=RCIM_Cubic, leaveTangent=-1, time=0.8, value=0.925}
         {arriveTangent=-2.125, interpMode=RCIM_Cubic, leaveTangent=-2.125, time=0.9, value=0.8}
         <at 1 0.5 RCIM_Cubic>
        ]}}}
     CurveIndex=$Particles.EscapeDistance
    )) {module=ScaleRibbonWidth001}
   &ScaleColor(
    Scale Alpha=&FloatFromCurve(
     FloatCurve=DataInterface() {propertyValues {Curve {keys [{tangentMode=RCTM_User, time=0, value=0}, {tangentMode=RCTM_User, time=1, value=1}]}}}
     CurveIndex=$Particles.BeamEndDissolve
    )
    ScaleA=true
    ScaleRGB=true
   ) {module=ScaleColor001}
   &ScaleColor(
    Scale Alpha=&Multiply_Float(
     A=&FloatFromCurve(
      FloatCurve=DataInterface() {propertyValues {Curve {keys [<at 0.7 1>, {tangentMode=RCTM_User, time=0.9, value=0.8}, {tangentMode=RCTM_User, time=1, value=0.5}]}}}
      CurveIndex=$Particles.EscapeDistance
     )
     B=&FNTimeOfDayScalar(Night Scale=0.4, Morning Scale=0.3, Evening Scale=0.5)
    )
    Scale RGB=&VectorFromFloat(Value=&MakeCustomFloatFromBool(
      Bool=&SetBoolByFloatComparison(
       A=&OneMinusFloat(Float=&FloatFromCurve(
         FloatCurve=DataInterface() {propertyValues {Curve {keys [<at 0.7 1>, {tangentMode=RCTM_User, time=0.9, value=0.8}, {tangentMode=RCTM_User, time=1, value=0.5}]}}}
         CurveIndex=$Particles.EscapeDistance
        ))
       B=0.2
      )
      True Float=&RandomRangeFloat(Minimum=0.5, Maximum=2)
      False Float=1
     ))
    ScaleA=true
    ScaleRGB=true
   )
   &ScaleRibbonWidth(Ribbon Width Scale=&FloatFromCurve(
     FloatCurve=DataInterface() {propertyValues {Curve {keys [
         {arriveTangent=0.509, interpMode=RCIM_Cubic, leaveTangent=0.509, tangentMode=RCTM_User, time=0, value=0}
         {arriveTangent=0.7269, interpMode=RCIM_Cubic, leaveTangent=0.7269, time=0.5, value=0.3}
         {arriveTangent=1.4, interpMode=RCIM_Cubic, leaveTangent=12.5702, tangentMode=RCTM_Break, time=0.7702, value=0.5598}
         {arriveTangent=3.7692, interpMode=RCIM_Cubic, leaveTangent=3.7692, tangentMode=RCTM_User, time=0.9979, value=2.1685}
        ]}}}
     CurveIndex=$Particles.Initial.Partition.NormalizedID
     Scale Curve=&Lerp_Float(A=0, B=5, Alpha=$User.OpenAmount)
    ))
   &ScaleColor(
    Scale Alpha=&FloatFromCurve(
     FloatCurve=DataInterface() {propertyValues {Curve {keys [
         {interpMode=RCIM_Cubic, tangentMode=RCTM_User, time=0, value=0}
         {arriveTangent=0.7047, interpMode=RCIM_Cubic, leaveTangent=0.7047, time=0.5, value=1}
         {arriveTangent=-4.3294, interpMode=RCIM_Cubic, leaveTangent=-4.3294, tangentMode=RCTM_User, time=0.8662, value=0.6104}
         <at 1 0 RCIM_Cubic>
        ]}}}
     CurveIndex=$Particles.Initial.Partition.NormalizedID
     Scale Curve=&Lerp_Float(A=0, B=2.5, Alpha=$User.OpenAmount)
    )
    ScaleA=true
    ScaleRGB=true
   ) {module=ScaleColor002}
  ]
  rendererClasses [&NiagaraRibbonRendererProperties]
  renderers [{rendererClass=&NiagaraRibbonRendererProperties}]
 }}
summary {
 systemName=NS_RiftGate_Umbramolt_Beam
 emitters [{emitterName=SplineElectric, rendererClasses [&NiagaraRibbonRendererProperties]}]
 userVariables [
  User.bIsFrontEnd {type=NiagaraBool}
  User.bIsFrontEndPreview {type=NiagaraBool}
  User.Growth {type=NiagaraFloat, defaultValue=1, description=""}
  User.OpenAmount {type=NiagaraFloat, defaultValue=1, description=""}
  User.OverallWidth {type=NiagaraFloat, defaultValue=150, description=""}
  User.RemoveBeam {type=NiagaraFloat, defaultValue=0, description=""}
  User.SplineUpdatedEndPos {type=Vector3f, description="", defaultValue {struct=Vector3f, value=<-25 -20 2500>}}
  User.SplineUpdatedStartPos {type=Vector3f, description="", defaultValue {struct=Vector3f, value=<-25 -20 350>}}
 ]
}
