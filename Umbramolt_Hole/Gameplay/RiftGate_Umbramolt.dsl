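; Umbramolt rift gate that runs the Gogo and Yuki twin boss encounter with Crazy 88 henchmen.
; Once playlist data is ready, the server spawns both twins through the Living World Manager, cross-links them and assigns intro roles.
; EQS places initial henchmen, each leashed to the gate and targeting the portal's player.
; Interaction knocks back players and adds an Activated tag, starting capped henchmen respawns on a timer until the Completed tag appears.
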(event Custom|InitiateTwinsSpawning
  ;[0.08133 0.0443 0.15 0.5] UmbraMolt_Twins
  (for (_array_element _array_index) (Utilities|Array|MakeArray (Variables|Default|GetGogoSpawnPoint) (Variables|Default|GetYukiSpawnPoint))
    (bind (_returnvalue_roll _returnvalue_pitch _returnvalue_yaw) (Math|Rotator|BreakRotator (Transformation|GetWorldRotation _array_element)))
    (Class|FortAthenaLivingWorldManager|TrySpawnEvent (Class|FortAthenaLivingWorldManager|GetLivingWorldManager) (Utilities|Struct|MakeDataTableRowHandle "/Lager/Lager/EventTables/Lager_EventTable_Boss_CH7S1.Lager_EventTable_Boss_CH7S1" (Utilities|Select :Index _array_index :Option%200 "Umbramolt_Fraternal_Y" :Option%201 "Umbramolt_Fraternal_G")) (Math|Transform|MakeTransform (Transformation|GetWorldLocation _array_element) (Math|Rotator|MakeRotator 0.0 0.0 _returnvalue_yaw)) (delegate Custom|TwinSpawned))))

(event Custom|TwinSpawned
  (params (SpawnedActors :type Array%20of%20Actor%20Object%20References) (bSuccess :type Boolean))
  ;[0.08133 0.0443 0.15 0.5] UmbraMolt_Twins
  (bind _asfort_pawn (Utilities|Casting|CastToFortPawn (Utilities|Array|Get(acopy) SpawnedActors)))
  (Class|UmbraMoltFraternalInterface|ToggleTwinState(Message) _asfort_pawn false)
  (bind _returnvalue (GameplayTags|HasMatchingGameplayTag(Message) _asfort_pawn "(TagName=\"AISpawnerData.Type.Boss.DragonCart.UmbraMolt_Fraternal.G\")"))
  (if _returnvalue
    (Variables|Default|SetGogoRef _asfort_pawn)
    (else
      (bind _returnvalue_2 (GameplayTags|HasMatchingGameplayTag(Message) _asfort_pawn "(TagName=\"AISpawnerData.Type.Boss.DragonCart.UmbraMolt_Fraternal.Y\")"))
      (if _returnvalue_2
        (Variables|Default|SetYukiRef _asfort_pawn))))
  (if (and (Utilities|IsValid (Variables|Default|GetGogoRef) :!class /Script/Engine.KismetSystemLibrary) (Utilities|IsValid (Variables|Default|GetYukiRef) :!class /Script/Engine.KismetSystemLibrary))
    (CallFunction|BothTwinSpawned)))

(event EventBeginPlay
  (|CallParentFunction)
  (Networking|SwitchHasAuthority
    (:Authority
      (Class|FortAsyncActionCurrentPlaylistReady|PlaylistDataReadyAsync
        (:OnReady
          (CallFunction|InitiateTwinsSpawning)
          (CallFunction|InitiateHenchmenSpawning))))))

(event Custom|BothTwinSpawned
  ;[0.08133 0.0443 0.15 0.5] UmbraMolt_Twins
  (CallFunction|ChooseRandomEmotesforWalkout)
  (Class|UmbraMoltFraternalInterface|PassTwinReference(Message) (Variables|Default|GetGogoRef) (Variables|Default|GetYukiRef))
  (Class|UmbraMoltFraternalInterface|PassTwinReference(Message) (Variables|Default|GetYukiRef) (Variables|Default|GetGogoRef))
  (bind _element (ArrayMacros|GetRandomArrayElement (Utilities|Array|MakeArray (Variables|Default|GetGogoRef) (Variables|Default|GetYukiRef)))
    (:%20%20
      (Class|UmbraMoltFraternalInterface|SetTalkatIntro(Message) _element true))))

(event EventEndPlay
  (params (EndPlayReason :type EEndPlayReason%20Enum :default "Destroyed"))
  (|CallParentFunction EndPlayReason)
  (Utilities|Time|ClearandInvalidateTimerbyHandle (Variables|Default|GetActivePortalHenchmenSpawn_TH)))

(event Custom|SpawnHenchmen
  ;[0.375 0.1356 0.02486 0.5] Crazy 88 Henchmen
  (if (not (>= (Utilities|Array|Length (Variables|Default|GetSpawnedHenchmen)) (ScalableFloat|GetValueasInteger (Variables|Hotfix|GetHenchmenMaxSpawnedAmount_Hf))))
    (bind _returnvalue (AI|EQS|RunEQSQuery "/Umbramolt_Hole/Gameplay/EQS/EQS_88Henchmen_SpawnLocations.EQS_88Henchmen_SpawnLocations" self))
    (AI|EQS|BindEventtoOnQueryFinished _returnvalue (delegate Custom|88HenchmenQueryCompleted))))

(event Custom|InitiateHenchmenSpawning
  ;[0.375 0.1356 0.02486 0.5] Crazy 88 Henchmen
  (Ability|Tasks|WaitGameplayTagAddToActor 0 "(TagName=\"UmbraMolt.Hole.Activated\")" true
    (:then
      (for _ (range 1 (+ (ScalableFloat|GetValueasInteger (Variables|Hotfix|GetHenchmenPreActivationSpawnAmnt_Hf)) 1))
        (bind _returnvalue (AI|EQS|RunEQSQuery "/Umbramolt_Hole/Gameplay/EQS/EQS_88Henchmen_SpawnLocations.EQS_88Henchmen_SpawnLocations" self))
        (AI|EQS|BindEventtoOnQueryFinished _returnvalue (delegate Custom|88HenchmenQueryCompleted))))
    (:Added
      (bind _returnvalue_2 (Utilities|Time|SetTimerbyEvent (EventDispatchers|CreateEvent :!event Custom|SpawnHenchmen) (ScalableFloat|GetValueatLevel (Variables|Hotfix|GetHenchmenRespawnTimer_Hf)) true true))
      (Variables|Default|SetActivePortalHenchmenSpawn_TH _returnvalue_2)
      (Ability|Tasks|WaitGameplayTagAddToActor 0 "(TagName=\"UmbraMolt.Hole.Completed\")" true
        (:Added
          (Utilities|Time|ClearandInvalidateTimerbyHandle (Variables|Default|GetActivePortalHenchmenSpawn_TH)))))))

(event Custom|88HenchmenQueryCompleted
  (params (QueryInstance :type EQS%20Query%20Instance%20Object%20Reference) (QueryStatus :type EEnvQueryStatus%20Enum :default "Processing"))
  ;[0.375 0.1356 0.02486 0.5] Crazy 88 Henchmen
  (bind (_resultlocations _returnvalue) (AI|EQS|GetQueryResultsasLocations QueryInstance))
  (if _returnvalue
    (Class|FortAthenaLivingWorldManager|TrySpawnEvent (Class|FortAthenaLivingWorldManager|GetLivingWorldManager) (Utilities|Struct|MakeDataTableRowHandle "/Lager/Lager/EventTables/Lager_EventTable_Minions_CH7S1.Lager_EventTable_Minions_CH7S1" "88_Minion") (Math|Transform|MakeTransform (Utilities|Array|Get(acopy) _resultlocations)) (delegate Custom|Crazy88HenchmenSpawned))))

(event Custom|Crazy88HenchmenSpawned
  (params (SpawnedActors :type Array%20of%20Actor%20Object%20References) (bSuccess :type Boolean))
  ;[0.375 0.1356 0.02486 0.5] Crazy 88 Henchmen
  (if bSuccess
    (bind _asfort_pawn (Utilities|Casting|CastToFortPawn (Utilities|Array|Get(acopy) SpawnedActors)))
    (Class|UmbraMoltFraternalInterface|PassPortalRef(Message) _asfort_pawn self)
    (Class|UmbraMoltFraternalInterface|PassBothTwinsReferences(Message) _asfort_pawn (Variables|Default|GetYukiRef) (Variables|Default|GetGogoRef))
    (Utilities|Array|AddUnique (Variables|Default|GetSpawnedHenchmen) _asfort_pawn)
    (bind _asbp_bot_controller_npc_base (Utilities|Casting|CastToBP_BotController_NPC_Base (Pawn|GetController _asfort_pawn)))
    (Class|BPFLAICommonFunctions|SetRadialLeash _asbp_bot_controller_npc_base (Transformation|GetActorLocation) (ScalableFloat|GetValueatLevel (Variables|Hotfix|GetHenchmenLeashInnerRadius_Hf)) (ScalableFloat|GetValueatLevel (Variables|Hotfix|GetHenchmenLeashOuterRadius_Hf)))
    (Class|FortAthenaAIBotController|AddTargetInfo _asbp_bot_controller_npc_base (Variables|RiftPortal|GetPlayerPawn) false "Threat_Seeing" 5.0 5e+03)))

(event EventSpawnedHenchmenEliminated
  (bind _spawnedhenchmen (Variables|Default|GetSpawnedHenchmen))
  ;[0.375 0.1356 0.02486 0.5] Crazy 88 Henchmen
  (for (_array_element _array_index) _spawnedhenchmen
    (if (Class|FortPawn|IsDying _array_element)
      (Utilities|Array|RemoveIndex _spawnedhenchmen _array_index))))

(event EventPostFinishSpawning)

(fn ChooseRandomEmotesForWalkout
  (bind _walkoutemoteindexes (Variables|Default|GetWalkoutEmoteIndexes))
  (bind (_outitem _outindex) (Utilities|Array|RandomArrayItem _walkoutemoteindexes))
  (bind _chosenwalkoutemoteindex (Variables|Default|GetChosenWalkoutEmoteIndex))
  (bind _output (Utilities|Array|Get(acopy) _walkoutemoteindexes _chosenwalkoutemoteindex))
  (bind _emotes (Class|UmbraMoltFraternalInterface|GetPostWalkoutEmotes(Message) (Variables|Default|GetGogoRef)))
  (for (_ _array_index) _emotes
    (Utilities|Array|Add (Variables|Default|GetWalkoutEmoteIndexes) _array_index))
  (for _array_element (Utilities|Array|MakeArray (Variables|Default|GetGogoRef) (Variables|Default|GetYukiRef))
    (Variables|Default|SetChosenWalkoutEmoteIndex _outindex)
    (Class|UmbraMoltFraternalInterface|SetWalkoutEmotetoPlay(Message) _array_element _output)
    (Utilities|Array|RemoveIndex (Variables|Default|GetWalkoutEmoteIndexes) _output)))

(fn GetIntroSpeechTag
  (return 0))

(fn GetPostWalkoutEmotes
  (return 0))

(fn GetTalkAtIntro
  (return false))

(fn GetTwinState
  (return false))

(fn GetWalkoutEmoteToPlay
  (return 0))

(fn PerformInteraction
  (bind _succeeded (|CallParentFunction))
  (if _succeeded
    (Development|PrintString "Spawning boss encounter!")
    (CallFunction|KnockBackPlayers)
    (CustomVerb|BroadcastCustomVerbMessage (CustomVerb|MakeCustomVerbMessage (Variables|RiftPortal|GetPlayerPawn) "(TagName=\"CustomVerb.UmbraMolt_Hole.RiftInteracted\")"))
    (Class|BPFLAICommonFunctions|AddUniqueGameplayTag self "(TagName=\"UmbraMolt.Hole.Activated\")")
    (return true)
    (else
      (return false))))

(fn ConstructionScript
  (|CallParentFunction)
  (for _array_element (Utilities|Array|MakeArray (Variables|Default|GetGogoSpawnPoint) (Variables|Default|GetYukiSpawnPoint))
    (Transformation|SetWorldLocation _array_element (AI|Navigation|ProjectPointtoNavigation (Transformation|GetWorldLocation _array_element) (Fortnite|AI|GetNavigationDataforActor self) "/Game/Athena/AI/Phoebe/NavFilter/NavFilter_Phoebe_AvoidWater.NavFilter_Phoebe_AvoidWater_C" "250.000000,250.000000,500.000000") false false)))
