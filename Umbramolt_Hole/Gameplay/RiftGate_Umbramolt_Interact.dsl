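; Umbramolt rift gate interactable that runs the Twins boss encounter.
; When the playlist is ready, the authority spawns Yuki and Gogo through the Living World Manager, leashes them, cross-links their references and assigns walkout emotes.
; Completed interaction tags the gate Activated, starting timed, capped EQS-placed henchman spawns until the Completed tag.
; It also launches overlapping vehicles away and drives boss music through an audio gameplay volume.
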(event OnComponentBeginOverlap(VehicleImpulseCollision)
  (params (OverlappedComponent :type Primitive%20Component%20Object%20Reference) (OtherActor :type Actor%20Object%20Reference) (OtherComp :type Primitive%20Component%20Object%20Reference) (OtherBodyIndex :type Integer) (bFromSweep :type Boolean) (SweepResult :type Hit%20Result%20Structure%20%28by%20ref%29))
  ;[1 0.1147 0] Impulse vehicles away
  (if (ScalableFloat|GetValueasBool (Variables|Hotfix|GetVehicleLaunchImpulse_Enabled_Hf))
    (bind _asfort_athena_vehicle (Utilities|Casting|CastToFortAthenaVehicle OtherActor))
    (Vehicle|WakeUp _asfort_athena_vehicle)
    (bind (_returnvalue_x _returnvalue_y) (Math|Vector|BreakVector (* (Math|Vector|Normalize (- (Transformation|GetActorLocation _asfort_athena_vehicle) (Transformation|GetActorLocation self))) (ScalableFloat|GetValueatLevel (Variables|Hotfix|GetVehicleLaunchImpulseXY_Hf)))))
    (Audio|BroadcastAbilityImpactatLocation 0 _asfort_athena_vehicle (Math|Vector|MakeVector _returnvalue_x _returnvalue_y (ScalableFloat|GetValueatLevel (Variables|Hotfix|GetVehicleLaunchImpulseZ_Hf))) 777.0 "None" false)))

(event Custom|InitiateTwinsSpawning
  ;[0.08133 0.0443 0.15 0.5] UmbraMolt_Twins
  (for (_array_element _array_index) (Utilities|Array|MakeArray (Variables|Default|GetYukiSpawnPoint) (Variables|Default|GetGogoSpawnPoint))
    (bind (_returnvalue_roll _returnvalue_pitch _returnvalue_yaw) (Math|Rotator|BreakRotator (Transformation|GetWorldRotation _array_element)))
    (Class|FortAthenaLivingWorldManager|TrySpawnEvent (Class|FortAthenaLivingWorldManager|GetLivingWorldManager) (Utilities|Struct|MakeDataTableRowHandle "/Lager/Lager/EventTables/Lager_EventTable_Boss_CH7S1.Lager_EventTable_Boss_CH7S1" (Utilities|Select :Index _array_index :Option%200 "Umbramolt_Fraternal_Y" :Option%201 "Umbramolt_Fraternal_G")) (Math|Transform|MakeTransform (Transformation|GetWorldLocation _array_element) (Math|Rotator|MakeRotator 0.0 0.0 _returnvalue_yaw)) (delegate Custom|TwinSpawned))))

(event EventBeginPlay
  (|CallParentFunction)
  (Networking|SwitchHasAuthority
    (:Authority
      (Class|FortAsyncActionCurrentPlaylistReady|PlaylistDataReadyAsync
        (:OnReady
          (CallFunction|InitiateTwinsSpawning)
          (CallFunction|InitiateHenchmenSpawning))))))

(event Custom|TwinSpawned
  (params (SpawnedActors :type Array%20of%20Actor%20Object%20References) (bSuccess :type Boolean))
  ;[0.08133 0.0443 0.15 0.5] UmbraMolt_Twins
  (bind _asfort_pawn (Utilities|Casting|CastToFortPawn (Utilities|Array|Get(acopy) SpawnedActors)))
  (Class|UmbraMoltFraternalInterface|ToggleTwinState(Message) _asfort_pawn false)
  (bind _asbp_bot_controller_npc_base (Utilities|Casting|CastToBP_BotController_NPC_Base (Pawn|GetController _asfort_pawn)))
  (Class|BPFLAICommonFunctions|SetRadialLeash _asbp_bot_controller_npc_base (Transformation|GetActorLocation) (ScalableFloat|GetValueatLevel (Variables|Hotfix|GetHenchmenLeashInnerRadius_Hf)) (ScalableFloat|GetValueatLevel (Variables|Hotfix|GetHenchmenLeashOuterRadius_Hf)))
  (Class|FortAthenaAIBotController|AddTargetInfo _asbp_bot_controller_npc_base (Variables|Default|GetLastPawnToInteract) false "Threat_Seeing" 5.0 5e+03)
  (bind _returnvalue (GameplayTags|HasMatchingGameplayTag(Message) _asfort_pawn "(TagName=\"AISpawnerData.Type.Boss.DragonCart.UmbraMolt_Fraternal.G\")"))
  (if _returnvalue
    (Variables|Default|SetGogoRef _asfort_pawn)
    (else
      (bind _returnvalue_2 (GameplayTags|HasMatchingGameplayTag(Message) _asfort_pawn "(TagName=\"AISpawnerData.Type.Boss.DragonCart.UmbraMolt_Fraternal.Y\")"))
      (if _returnvalue_2
        (Variables|Default|SetYukiRef _asfort_pawn))))
  (if (and (Utilities|IsValid (Variables|Default|GetGogoRef) :!class /Script/Engine.KismetSystemLibrary) (Utilities|IsValid (Variables|Default|GetYukiRef) :!class /Script/Engine.KismetSystemLibrary))
    (CallFunction|BothTwinSpawned)))

(event EventEndPlay
  (params (EndPlayReason :type EEndPlayReason%20Enum :default "Destroyed"))
  (|CallParentFunction EndPlayReason)
  (Utilities|Time|ClearandInvalidateTimerbyHandle (Variables|Default|GetActivePortalHenchmenSpawn_TH))
  (bind _musicaudiocomponent (Variables|Default|GetMusicAudioComponent)
    (:then
      (Audio|Parameter|ExecuteTriggerParameter _musicaudiocomponent "OnDefeated"))
    (:else)))

(event Custom|BothTwinSpawned
  ;[0.08133 0.0443 0.15 0.5] UmbraMolt_Twins
  (Default|ChooseRandomEmotesforWalkout)
  (Class|UmbraMoltFraternalInterface|PassTwinReference(Message) (Variables|Default|GetGogoRef) (Variables|Default|GetYukiRef))
  (Class|UmbraMoltFraternalInterface|PassTwinReference(Message) (Variables|Default|GetYukiRef) (Variables|Default|GetGogoRef))
  (bind _element (ArrayMacros|GetRandomArrayElement (Utilities|Array|MakeArray (Variables|Default|GetGogoRef) (Variables|Default|GetYukiRef)))
    (:%20%20
      (Class|UmbraMoltFraternalInterface|SetTalkatIntro(Message) _element true))))

(event Custom|SpawnHenchmen
  ;[0.375 0.1356 0.02486 0.5] Crazy 88 Henchmen
  (if (not (>= (Utilities|Array|Length (Variables|Default|GetSpawnedHenchmen)) (ScalableFloat|GetValueasInteger (Variables|Hotfix|GetHenchmenMaxSpawnedAmount_Hf))))
    (bind _returnvalue (AI|EQS|RunEQSQuery "/Umbramolt_Hole/Gameplay/EQS/EQS_88Henchmen_SpawnLocations.EQS_88Henchmen_SpawnLocations" self))
    (AI|EQS|BindEventtoOnQueryFinished _returnvalue (delegate Custom|88HenchmenQueryCompleted))))

(event Custom|InitiateHenchmenSpawning
  ;[0.375 0.1356 0.02486 0.5] Crazy 88 Henchmen
  (Ability|Tasks|WaitGameplayTagAddToActor 0 "(TagName=\"UmbraMolt.Hole.Activated\")" true
    (:Added
      (bind _returnvalue (Utilities|Time|SetTimerbyEvent (EventDispatchers|CreateEvent :!event Custom|SpawnHenchmen) (ScalableFloat|GetValueatLevel (Variables|Hotfix|GetHenchmenRespawnTimer_Hf)) true true))
      (Variables|Default|SetActivePortalHenchmenSpawn_TH _returnvalue)
      ;;[1 1 1] Removed pre-interact spawn
      (Ability|Tasks|WaitGameplayTagAddToActor 0 "(TagName=\"UmbraMolt.Hole.Completed\")" true
        (:Added
          (Utilities|Time|ClearandInvalidateTimerbyHandle (Variables|Default|GetActivePortalHenchmenSpawn_TH)))))))

(event Custom|88HenchmenQueryCompleted
  (params (QueryInstance :type EQS%20Query%20Instance%20Object%20Reference) (QueryStatus :type EEnvQueryStatus%20Enum :default "Processing"))
  ;[0.375 0.1356 0.02486 0.5] Crazy 88 Henchmen
  (bind (_resultlocations _returnvalue) (AI|EQS|GetQueryResultsasLocations QueryInstance))
  (if _returnvalue
    (Class|FortAthenaLivingWorldManager|TrySpawnEvent (Class|FortAthenaLivingWorldManager|GetLivingWorldManager) (Utilities|Struct|MakeDataTableRowHandle "/Lager/Lager/EventTables/Lager_EventTable_Minions_CH7S1.Lager_EventTable_Minions_CH7S1" "88_Minion") (Math|Transform|MakeTransform (Utilities|Array|Get(acopy) _resultlocations)) (delegate Custom|Crazy88HenchmenSpawned))))

(event Custom|Crazy88HenchmenSpawned
  (params (SpawnedActors :type Array%20of%20Actor%20Object%20References) (bSuccess :type Boolean))
  ;[0.375 0.1356 0.02486 0.5] Crazy 88 Henchmen
  (if bSuccess
    (bind _asfort_pawn (Utilities|Casting|CastToFortPawn (Utilities|Array|Get(acopy) SpawnedActors)))
    (Class|UmbraMoltFraternalInterface|PassPortalRef(Message) _asfort_pawn self)
    (Class|UmbraMoltFraternalInterface|PassBothTwinsReferences(Message) _asfort_pawn (Variables|Default|GetYukiRef) (Variables|Default|GetGogoRef))
    (Utilities|Array|AddUnique (Variables|Default|GetSpawnedHenchmen) _asfort_pawn)
    (bind _asbp_bot_controller_npc_base (Utilities|Casting|CastToBP_BotController_NPC_Base (Pawn|GetController _asfort_pawn)))
    (Class|BPFLAICommonFunctions|SetRadialLeash _asbp_bot_controller_npc_base (Transformation|GetActorLocation) (ScalableFloat|GetValueatLevel (Variables|Hotfix|GetHenchmenLeashInnerRadius_Hf)) (ScalableFloat|GetValueatLevel (Variables|Hotfix|GetHenchmenLeashOuterRadius_Hf)))
    (Class|FortAthenaAIBotController|AddTargetInfo _asbp_bot_controller_npc_base (Variables|Default|GetLastPawnToInteract) false "Threat_Seeing" 5.0 5e+03)))

(event Custom|CustomEvent
  (bind _spawnedhenchmen (Variables|Default|GetSpawnedHenchmen))
  ;[0.375 0.1356 0.02486 0.5] Crazy 88 Henchmen
  (for (_array_element _array_index) _spawnedhenchmen
    (if (Class|FortPawn|IsDying _array_element)
      (Utilities|Array|RemoveIndex _spawnedhenchmen _array_index))))

(event EventPlayRiftMusic
  (params (MusicStinger :type Sound%20Base%20Object%20Reference))
  ;[0 0 0 0.5] Music Implementation
  (Components|Activation|Activate (Variables|Default|GetAudioGameplayVolume) false)
  (Networking|CanExecuteCosmeticEvents
    (:True
      (CallFunction|PlaySequenceMusic)
      (bind _spawned_riftportalsoundloop (Variables|Default|GetSpawned_RiftPortalSoundLoop)
        (:then
          (Audio|Components|Audio|FadeOut _spawned_riftportalsoundloop 2.0 0.0 "Linear" :!class /Script/Engine.AudioComponent))
        (:else)))))

(event OnToggledOn(AudioGameplayVolume)
  ;[0 0 0 0.5] Music Implementation
  (if (and (Utilities|Array|IsNotEmpty (Variables|Default|GetSpawnedHenchmen)) (Utilities|Enum|Equal(Enum) (Variables|RiftGate|State|GetRiftGateState_Current) "NewEnumerator2"))
    (CallFunction|PlaySequenceMusic)))

(event OnToggledOff(AudioGameplayVolume)
  ;[0 0 0 0.5] Music Implementation
  (bind _musicaudiocomponent (Variables|Default|GetMusicAudioComponent)
    (:then
      (Audio|Parameter|SetBooleanParameter _musicaudiocomponent "MusicEncounter.bIsPaused" true :!class /Script/AudioExtensions.AudioParameterControllerInterface))
    (:else)))

(event Custom|PlaySequenceMusic
  ;[0 0 0 0.5] Music Implementation
  (Variables|Default|GetMusicAudioComponent
    (:then
      _)
    (:else
      (bind _returnvalue (Audio|SpawnSoundatLocation "/Umbramolt_Hole/Assets/Audio/Music/MSS_UmbraMolt_Music.MSS_UmbraMolt_Music" (Transformation|GetActorLocation)))
      (bind _output_get (Variables|Default|SetMusicAudioComponent _returnvalue))
      (Audio|AddEventbyComponent (GameInstanceSubsystems|GetMusicEventSubsystem) _output_get "(TagName=\"MusicEventSubsystem.Event.Boss\")" "(GameplayTags=)" 0 "")))
  (Audio|Parameter|SetBooleanParameter _output_get "MusicEncounter.bIsPaused" false :!class /Script/AudioExtensions.AudioParameterControllerInterface))

(fn ChooseRandomEmotesForWalkout
  (bind _walkoutemoteindexes (Variables|Default|GetWalkoutEmoteIndexes))
  (bind (_outitem _outindex) (Utilities|Array|RandomArrayItem _walkoutemoteindexes))
  (bind _chosenwalkoutemoteindex (Variables|Default|GetChosenWalkoutEmoteIndex))
  (bind _output (Utilities|Array|Get(acopy) _walkoutemoteindexes _chosenwalkoutemoteindex))
  (bind _emotes (Class|UmbraMoltFraternalInterface|GetPostWalkoutEmotes(Message) (Variables|Default|GetGogoRef)))
  (for (_ _array_index) _emotes
    (Utilities|Array|Add (Variables|Default|GetWalkoutEmoteIndexes) _array_index))
  (for _array_element (Utilities|Array|MakeArray (Variables|Default|GetGogoRef) (Variables|Default|GetYukiRef))
    (Variables|Default|SetChosenWalkoutEmoteIndex _outindex)
    (Class|UmbraMoltFraternalInterface|SetWalkoutEmotetoPlay(Message) _array_element _output)
    (Utilities|Array|RemoveIndex (Variables|Default|GetWalkoutEmoteIndexes) _output)))

(fn OnRep_bHasInteracted
  (Networking|CanExecuteCosmeticEvents))

(fn OnRiftGateInteractionComplete
  (|CallParentFunction)
  (Class|BPFLAICommonFunctions|AddUniqueGameplayTag self "(TagName=\"UmbraMolt.Hole.Activated\")")
  (CustomVerb|BroadcastCustomVerbMessage (CustomVerb|MakeCustomVerbMessage (Variables|Default|GetLastPawnToInteract) "(TagName=\"CustomVerb.UmbraMolt_Hole.RiftInteracted\")")))

(fn ConstructionScript
  (|CallParentFunction)
  (for _array_element (Utilities|Array|MakeArray (Variables|Default|GetGogoSpawnPoint) (Variables|Default|GetYukiSpawnPoint))
    (Transformation|SetWorldLocation _array_element (AI|Navigation|ProjectPointtoNavigation (Transformation|GetWorldLocation _array_element) (Fortnite|AI|GetNavigationDataforActor self) "/Game/Athena/AI/Phoebe/NavFilter/NavFilter_Phoebe_AvoidWater.NavFilter_Phoebe_AvoidWater_C" "250.000000,250.000000,500.000000") false false)))
